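; Part modifier Blueprint for the DriedSilk backpack cosmetic, handling frontend toon shading tweaks and kill-reactive VFX.
; After PostResIn it runs a toon LUT replacement, adjusts material camera-blend shading when frontend reactivity displays, and binds to the backpack AnimBP's reactivity event.
; An active kill activates the idle VFX with Reactive set, then deactivates it after a retriggerable delay; otherwise Reactive clears.
; The older stat-change reactivity path only calls the parent.
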(event EventOnPostResIn
  (|CallParentFunction)
  (Utilities|FlowControl|Sequence
    (:then_0
      (Utilities|ToonFrontendLUTReplace
        (:Out
          (if (Reactive|Frontend|ShouldDisplayFrontendReactivity 0 "Backpack")
            ;[1 1 1] Hatching FE Adjustments
            (if (or (Fortnite|IsMobilePlatform) (Class|PlatformHelpersLibrary|IsPcPerfMode))
              _)
            (Rendering|Material|SetScalarParameterValueonMaterials (CustomCharacter|GetSkeletalMeshForPartOnPlayerPawn) "ShadingCurve_CameraBlend" 0.5)))))
    (:then_1
      (bind _asskeletal_mesh_component (Utilities|Casting|CastToSkeletalMeshComponent (CustomCharacter|GetSkeletalMeshForPartOnPlayerPawn)))
      (bind _asm_med_dried_silk_pack_anim_bp (Utilities|Casting|CastToM_MED_DriedSilk_Pack_AnimBP (Components|SkeletalMesh|GetAnimInstance _asskeletal_mesh_component)))
      (Default|BindEventtoUpdateReactivityFX _asm_med_dried_silk_pack_anim_bp (delegate Custom|ReactivityFX)))))

(event Custom|ReactivityFX
  (params (ActiveKill :type Boolean))
  (bind _returnvalue (CallFunction|GetIdleVFXComponent))
  ;[1 0.6073 0.1674] New FX Reactivity Implementation
  (if ActiveKill
    (Components|Activation|SetActive _returnvalue true false)
    (Effects|Components|ParticleSystem|SetBoolParameter _returnvalue "Reactive" true)
    (Utilities|FlowControl|RetriggerableDelay 2.0)
    (Components|Activation|SetActive _returnvalue false false)
    (else
      (Effects|Components|ParticleSystem|SetBoolParameter _returnvalue "Reactive" false))))

(event EventOnStatChanged
  (params (State%20Name :type Name :default "None") (Stat%20Value :type Integer))
  ;[1 1 1] Old FX Reactivity
  (|CallParentFunction State%20Name Stat%20Value))

(fn ConstructionScript
  (|CallParentFunction))
