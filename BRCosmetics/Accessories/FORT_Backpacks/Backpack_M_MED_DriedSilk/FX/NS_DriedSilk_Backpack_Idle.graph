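# Niagara idle effect for the DriedSilk backpack cosmetic, rendering a propeller mesh from a single emitter.
# A burst spawns a small flattened mesh particle at an offset from the system, with a fixed local orientation, kept alive past its lifetime.
# Dynamic material parameters include a User.Reactivity-driven lerp.
# Alpha follows the system timeline's normalized time, dimmed in the front end; the Reactivity-based alpha scale is disabled.
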NiagaraSystem /BRCosmetics/Accessories/FORT_Backpacks/Backpack_M_MED_DriedSilk/FX/NS_DriedSilk_Backpack_Idle
dependencies {usedDataInterfaces [], usedDynamicInputs [], usedModules [&SystemState, &Timeline]}
emitters {Propeller {
  emitterUpdateScript [
   &EmitterState()
   &SpawnBurst_Instantaneous(Use Spawn Probability=false, UseLoopCountLimit=true)
  ]
  particleSpawnScript [
   &InitializeParticle(
    Lifetime=2
    Mesh Scale=&Multiply_VectorByFloat(Vector=<0.42 0.42 0.1>, Float=0.1)
    Mesh Scale Mode=::NewEnumerator4
    UsePositionOffset=false
    Write Material Random=true
   )
   &SystemLocation(MaskSpawnGroup=false, Offset=<0 0 0.3>)
   &InitialMeshOrientation(
    Mesh Axis To Orient=<1 0 0.1>
    Mesh Orientation Mode=::NewEnumerator2
    Orientation Coordinate Space=::Local
    Rotation=<0 0.5 0>
    Use Rotation Vector=true
   )
   /SetVariables(Particles.MeshOrientation=&MultiplyQuaternion(Quaternion A=$Particles.MeshOrientation, Quaternion B=<0 0 1 1>)) {isSetParameters=true}
  ]
  particleUpdateScript [
   &ParticleState(Kill Particles When Lifetime Has Elapsed=false, Let Infinitely Lived Particles Die When Emitter Deactivates=true)
   &DynamicMaterialParameters(
    Index 0 Param 1=0.12
    Index 0 Param 2=&Lerp_Float(A=1, B=0.96, Alpha=$User.Reactivity)
    Module.Index0Param0Write=true
    Module.Index0Param1Write=true
    Module.Index0Param2Write=false
    Module.Index0Param3Write=false
   )
   &ScaleColor(Scale Alpha=$User.Reactivity, ScaleA=true, ScaleRGB=false) {enabled=false}
   &ScaleColor(
    Scale Alpha=&Multiply_Float(A=$System.Timeline.NormalizedTime, B=&FN_FrontEndFloatModifier(Gameplay=1, Front End=0.7))
    ScaleA=true
    ScaleRGB=false
   ) {module=ScaleColor001}
  ]
  rendererClasses [&NiagaraMeshRendererProperties]
  renderers [{rendererClass=&NiagaraMeshRendererProperties, rendererIndex=0}]
 }}
summary {
 systemName=NS_DriedSilk_Backpack_Idle
 emitters [{emitterName=Propeller, rendererClasses [&NiagaraMeshRendererProperties]}]
 userVariables [
  User.bIsFrontEnd {type=NiagaraBool}
  User.bIsFrontEndPreview {type=NiagaraBool}
  User.Reactive {type=NiagaraBool}
  User.Reactivity {type=NiagaraFloat, defaultValue=0.5, description=""}
 ]
}
